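# Physics asset for the front-end face accessory mesh of the female medium Shatter Fly Eclipse cosmetic body, setting up physics for its dynamic hair bones.
# Capsule bodies cover the head, neck, upper spine, left clavicle and hair bones, with free angular motion between the core body bones.
# Three single hair bones and a four-bone dyn_hair_05 chain hang from the head, each joint with its own swing and twist limits.
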PhysicsAsset /BRCosmetics/Characters/Player/Female/Medium/Bodies/F_MED_Shatter_Fly_Eclipse/Meshes/Parts/F_MED_Shatter_Fly_Eclipse_FaceAcc_FrontEnd_Physics
bodies {
 clavicle_l [{shapeType=Capsule, length=10, radius=6.5, shapeName=clavicle_l_capsule, center=<3.8884 -1.1162 -1.1292>, rotation=<rot 80 -0.0007 -0.0008>}]
 dyn_hair_01 [{shapeType=Capsule, length=4, radius=0.5, shapeName=dyn_hair_01_capsule, center=<1.64 0 0>, rotation=<rot 90 0 0>}]
 dyn_hair_02 [{shapeType=Capsule, length=2.8, radius=0.5, shapeName=dyn_hair_02_capsule, center=<1 -0.7 0>, rotation=<rot 90 0 38.32>}]
 dyn_hair_03 [{shapeType=Capsule, length=4.2, radius=0.9, shapeName=dyn_hair_03_capsule, center=<1.9 0 0>, rotation=<rot 90 0 0>}]
 dyn_hair_05 [{shapeType=Capsule, length=3, radius=3, shapeName=dyn_hair_05_capsule, center=<-2.4892 -2.04 0>}]
 dyn_hair_05_a [{shapeType=Capsule, length=3, radius=2, shapeName=dyn_hair_05_a_capsule, center=<1.8779 0 0>, rotation=<rot 90 0 0>}]
 dyn_hair_05_b [{shapeType=Capsule, length=3, radius=3, shapeName=dyn_hair_05_b_capsule, center=<0.8122 0 0.8027>, rotation=<rot 80.0002 0 -0.0001>}]
 dyn_hair_05_c [{shapeType=Capsule, length=3, radius=2, shapeName=dyn_hair_05_c_capsule, center=<1.39 0 0.6105>, rotation=<rot 90 0 0>}]
 head [{shapeType=Capsule, length=5, radius=5, shapeName=head_capsule, rotation=<rot 90 0 0>}]
 neck_01 [{shapeType=Capsule, length=5, radius=4, shapeName=neck_01_capsule, rotation=<rot 90 0 0>}]
 spine_05 [{shapeType=Capsule, length=25, radius=8, shapeName=spine_05_capsule, center=<1.0185 -0.8174 0>}]
}
constraints [
 {
  bone1Name=clavicle_l
  bone2Name=spine_05
  swing1LimitDegrees=45
  swing1Motion=Free
  swing2LimitDegrees=45
  swing2Motion=Free
  twistLimitDegrees=45
  twistMotion=Free
 }
 {
  bone1Name=head
  bone2Name=neck_01
  swing1LimitDegrees=45
  swing1Motion=Free
  swing2LimitDegrees=45
  swing2Motion=Free
  twistLimitDegrees=45
  twistMotion=Free
 }
 {bone1Name=dyn_hair_05, bone2Name=head, swing1LimitDegrees=50, swing2LimitDegrees=5, twistLimitDegrees=50}
 {bone1Name=dyn_hair_01, bone2Name=head, swing1LimitDegrees=15, swing2LimitDegrees=15, twistLimitDegrees=45}
 {bone1Name=dyn_hair_02, bone2Name=head, swing1LimitDegrees=15, swing2LimitDegrees=15, twistLimitDegrees=45}
 {bone1Name=dyn_hair_03, bone2Name=head, swing1LimitDegrees=15, swing2LimitDegrees=15, twistLimitDegrees=15}
 {bone1Name=dyn_hair_05_a, bone2Name=dyn_hair_05, swing1LimitDegrees=20, swing2LimitDegrees=20, twistLimitDegrees=10}
 {bone1Name=dyn_hair_05_b, bone2Name=dyn_hair_05_a, swing1LimitDegrees=30, swing2LimitDegrees=30, twistLimitDegrees=10}
 {bone1Name=dyn_hair_05_c, bone2Name=dyn_hair_05_b, swing1LimitDegrees=10, swing2LimitDegrees=30, twistLimitDegrees=1}
 {
  bone1Name=neck_01
  bone2Name=spine_05
  swing1LimitDegrees=45
  swing1Motion=Free
  swing2LimitDegrees=45
  swing2Motion=Free
  twistLimitDegrees=45
  twistMotion=Free
 }
]
