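# Post-process style animation blueprint for the Shatter Fly Eclipse face accessory, animating its dynamic hair on a pose copied from the attached parent mesh.
# In the front end, weighted constraints tie hair bones to spine and left clavicle before head-relative RigidBody simulation.
# Otherwise, a look-at and additive offset, scaled by spine-to-clavicle delta computed each update, feed a RigidBody with its own physics asset.
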AnimBlueprint /BRCosmetics/Characters/Player/Female/Medium/Bodies/F_MED_Shatter_Fly_Eclipse/Meshes/Parts/F_MED_Shatter_Fly_Eclipse_FaceAcc_AnimBP
target_skeleton=&/F_MED_Shatter_Fly_Eclipse_FaceAcc_Skeleton
graphs [
 AnimGraph {nodes [
   n1 = Look At - Bone: dyn_hair_05_d(ComponentPose=Local To Component(LocalPose=Use cached pose 'StartCache'()), Alpha=1) {
    alphaBoolBlend {blendOption=Linear}
    alphaScaleBiasClamp {bInterpResult=true}
    boneToModify {boneName=dyn_hair_05_d}
    lookAtLocation=<18 20 -10>
    lookAtTarget {boneReference {boneName=Target_clavicle_l}}
   }
   n2 = "Transform (Modify) Bone - Bone: dyn_hair_05"(ComponentPose=@n1, Alpha=Get Hair Up() -> (Hair Up:double), Translation="-6,-2,2", Rotation="30,10,50", Scale="1,1,1") {
    rotationMode=BMM_Additive
    rotationSpace=BCS_BoneSpace
    translationMode=BMM_Additive
    translationSpace=BCS_BoneSpace
    alphaBoolBlend {blendOption=Linear}
    boneToModify {boneName=dyn_hair_05}
    rotation=<rot 30 10 50>
    translation=<-6 -2 2>
   }
   n3 = RigidBody(ComponentPose=@n2) {
    bClampLinearTranslationLimitToRefPose=true
    bOverrideWorldGravity=true
    overridePhysicsAsset=&/F_MED_Shatter_Fly_Eclipse_FaceAcc_Physics
    simulationSpace=BaseBoneSpace
    alphaBoolBlend {blendOption=Linear}
    baseBoneRef {boneName=head}
    componentAppliedLinearAccClamp=<1000 1000 1000>
    componentLinearAccScale=<0.01 0.01 0.01>
    componentLinearVelScale=<0.01 0.01 0.01>
    overrideWorldGravity {z=-500}
   }
   n4 = "Constraint - dyn_hair_05_a (PPP)"!(
    ComponentPose=Local To Component(LocalPose=Use cached pose 'StartCache'())
    Alpha=1
    ConstraintWeights_0=0.5617
    ConstraintWeights_1:float=0.4239
    ConstraintWeights_2:float=0.0143
   ) -> (Pose) {
    alphaBoolBlend {blendOption=Linear}
    boneToModify {boneName=dyn_hair_05_a}
    constraintSetup [{targetBone {boneName=spine_05}}, {targetBone {boneName=clavicle_l}}, {targetBone {boneName=clavicle_l}}]
    constraintWeights [1,1,1]
   }
   n5 = "Constraint - dyn_hair_05_b (PPPP)"!(
    ComponentPose=@n4
    Alpha=1
    ConstraintWeights_0=0.505
    ConstraintWeights_1:float=0.4409
    ConstraintWeights_2:float=0.0394
    ConstraintWeights_3:float=0.0148
   ) -> (Pose) {
    alphaBoolBlend {blendOption=Linear}
    boneToModify {boneName=dyn_hair_05_b}
    constraintSetup [{targetBone {boneName=spine_05}}, {targetBone {boneName=clavicle_l}}, {targetBone {boneName=spine_04}}, {targetBone {boneName=clavicle_l}}]
    constraintWeights [1,1,1,1]
   }
   n6 = "Constraint - dyn_hair_05_c (PPPP)"!(
    ComponentPose=@n5
    Alpha=1
    ConstraintWeights_0=0.4488
    ConstraintWeights_1:float=0.3883
    ConstraintWeights_2:float=0.1378
    ConstraintWeights_3:float=0.0252
   ) -> (Pose) {
    alphaBoolBlend {blendOption=Linear}
    boneToModify {boneName=dyn_hair_05_c}
    constraintSetup [{targetBone {boneName=clavicle_l}}, {targetBone {boneName=spine_05}}, {targetBone {boneName=spine_04}}, {targetBone {boneName=clavicle_l}}]
    constraintWeights [1,1,1,1]
   }
   n7 = RigidBody(ComponentPose=@n6) {
    bClampLinearTranslationLimitToRefPose=true
    bOverrideWorldGravity=true
    overridePhysicsAsset=&/F_MED_Shatter_Fly_Eclipse_FaceAcc_FrontEnd_Physics
    simulationSpace=BaseBoneSpace
    alphaBoolBlend {blendOption=Linear}
    baseBoneRef {boneName=head}
    componentAppliedLinearAccClamp=<1000 1000 1000>
    componentLinearAccScale=<0.01 0.01 0.01>
    componentLinearVelScale=<0.01 0.01 0.01>
    overrideWorldGravity {z=-500}
   }
   Output Pose(Result=Blend Poses by bool(BlendPose_0=Component To Local(ComponentPose=@n7), BlendPose_1=Component To Local(ComponentPose=@n3), bActiveValue=Get bIsInFrontEnd()) {blendType=Linear}) {layerGroup=None}
   Save cached pose 'StartCache'(Pose=Copy Pose From Mesh() {bUseAttachedParent=true}) {cacheName=StartCache}
  ]}
 EventGraph {nodes [
   Event BlueprintUpdateAnimation()
   > Set Hair Up(Hair Up:double=GetAlphaFromReferenceTransformDelta(SkeletalMeshComponent=Get Skel Mesh(), BoneName=spine_05, BaseName=clavicle_l, InputRange=(X=-5,Y=-15), TwistAxis="0,0,1")) -> (Output_Get:double)
   Event BlueprintInitializeAnimation()
   > Set Skel Mesh(Skel Mesh=GetOwnerSkeletalMeshComponent())
  ]}
]
variables [
 Hair Up {type=number}
 Skel Mesh {type=SkeletalMeshComponent}
]
